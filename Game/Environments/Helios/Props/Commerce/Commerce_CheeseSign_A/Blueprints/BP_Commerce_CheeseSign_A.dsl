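; BP_Commerce_CheeseSign_A is a Blueprint for a cheese sign prop in the Helios commerce environment set.
; Its BeginPlay, overlap, Tick and OnMeshSetChanged_Event events are empty stubs, the last taking a Building Static Mesh object and a Mesh Set.
; The construction script only calls the parent, so the Blueprint adds no graph logic of its own.
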(event EventBeginPlay)

(event Collision|EventActorBeginOverlap
  (params (OtherActor :type Actor%20Object%20Reference)))

(event EventTick
  (params (DeltaSeconds :type Float%20%28single-precision%29)))

(event Custom|OnMeshSetChanged_Event
  (params (BuildingSMActor :type Building%20Static%20Mesh%20Object%20Reference) (NewMeshSet :type Mesh%20Set%20Structure%20%28by%20ref%29)))

(fn ConstructionScript
  (|CallParentFunction))
